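# Looping ambient spot cue for large glacier creaks in the RealityWarrior biome, routed to the outdoor ambient sound class.
# A loop waits a random delay of several seconds, then plays one of seven equally weighted creak waves with slight pitch and volume modulation through capsule linear attenuation.
# A quality-level node feeds that shared chain to its first two slots and leaves the other two empty, so those levels play no creak.
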SoundCue /Game/Sounds/Asteria/Biomes/RealityWarrior/Spots/Spot_RealityWarrior_Glacier_Lge_Creak_Looping_Cue
root=Looping_0
nodes [
 {
  id=Attenuation_0
  children [Random_2]
  properties {attenuationSettings=&/Spot_Glacier_Creak_Capsule_Linear_Att}
 }
 {id=Delay_0, children [QualityLevel_0], properties {delayMax=22, delayMin=8}}
 {id=Looping_0, children [Delay_0]}
 {id=Modulator_1, children [Attenuation_0], properties {pitchMax=1.1, pitchMin=0.9, volumeMin=0.85}}
 {id=QualityLevel_0, children [Modulator_1,Modulator_1,null,null]}
 {
  id=Random_2
  children [WavePlayer_15, WavePlayer_16, WavePlayer_17, WavePlayer_18, WavePlayer_19, WavePlayer_20, WavePlayer_21]
  properties {weights [1,1,1,1,1,1,1]}
 }
 {id=WavePlayer_15, properties {soundWaveAssetPtr=&/Spot_Glacier_Creak_01}}
 {id=WavePlayer_16, properties {soundWaveAssetPtr=&/Spot_Glacier_Creak_02}}
 {id=WavePlayer_17, properties {soundWaveAssetPtr=&/Spot_Glacier_Creak_03}}
 {id=WavePlayer_18, properties {soundWaveAssetPtr=&/Spot_Glacier_Creak_04}}
 {id=WavePlayer_19, properties {soundWaveAssetPtr=&/Spot_Glacier_Creak_05}}
 {id=WavePlayer_20, properties {soundWaveAssetPtr=&/Spot_Glacier_Creak_06}}
 {id=WavePlayer_21, properties {soundWaveAssetPtr=&/Spot_Glacier_Creak_07}}
]
properties {
 attenuationSettings=&Spot_Glacier_Creak_Capsule_Linear_Att
 maxDistance=5000
 soundClassObject=&Fort_Ambient_Outdoors_Athena
 volumeMultiplier=0.501
 preEffectBusSends [{}]
}
