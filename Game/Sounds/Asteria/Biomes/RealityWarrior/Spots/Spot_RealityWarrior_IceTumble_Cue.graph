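# Sound cue for an ice-tumble ambient spot one-shot in the RealityWarrior biome, reusing Winter weather one-shot waves.
# A random node picks one of five equally weighted ice-tumble variations, then a modulator varies the result.
# It plays at reduced volume through outdoor ambient one-shot class and tile-ambient wind attenuation, and sends to the exterior ambient source bus.
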SoundCue /Game/Sounds/Asteria/Biomes/RealityWarrior/Spots/Spot_RealityWarrior_IceTumble_Cue
root=Modulator_1
nodes [
 {id=Modulator_1, children [Random_1]}
 {id=Random_1, children [WavePlayer_0,WavePlayer_6,WavePlayer_7,WavePlayer_8,WavePlayer_9], properties {weights [1,1,1,1,1]}}
 {id=WavePlayer_0, properties {soundWaveAssetPtr=/Game/Sounds/Apollo/Weather/Winter/OneShots/Winter_Oneshot_IceTumble_05}}
 {id=WavePlayer_6, properties {soundWaveAssetPtr=/Game/Sounds/Apollo/Weather/Winter/OneShots/Winter_Oneshot_IceTumble_04}}
 {id=WavePlayer_7, properties {soundWaveAssetPtr=/Game/Sounds/Apollo/Weather/Winter/OneShots/Winter_Oneshot_IceTumble_03}}
 {id=WavePlayer_8, properties {soundWaveAssetPtr=/Game/Sounds/Apollo/Weather/Winter/OneShots/Winter_Oneshot_IceTumble_02}}
 {id=WavePlayer_9, properties {soundWaveAssetPtr=/Game/Sounds/Apollo/Weather/Winter/OneShots/Winter_Oneshot_IceTumble_01}}
]
properties {
 attenuationSettings=/Game/Sounds/Attenuation/Fort_TileAmbient_OneShot_Wind_Att
 maxDistance=8000
 soundClassObject=&Fort_Ambient_OneShots_Outdoors
 volumeMultiplier=0.501
 busSends [{audioBus=null, minSendLevel=0.6, sendLevel=1, soundSourceBus=&AmbientAudio_Exterior_SourceBus}]
}
