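# Sound cue for close ice-pressure spot one-shots in the RealityWarrior biome's outdoor ambience.
# A Random node picks one of eight equally weighted ice pressure waves, then a Modulator varies volume and applies a wide, mostly downward pitch shift.
# It uses linear tile-ambient one-shot attenuation, plays in the outdoor ambient one-shots sound class at reduced volume, and sends to the distant reverb submix.
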SoundCue /Game/Sounds/Asteria/Biomes/RealityWarrior/Spots/Spot_RealityWarrior_Ice_Pressure_Close_Cue
root=Modulator_0
nodes [
 {id=Modulator_0, children [Random_0], properties {pitchMax=0.5, pitchMin=-1.5, volumeMin=0.85}}
 {
  id=Random_0
  children [WavePlayer_1, WavePlayer_2, WavePlayer_3, WavePlayer_4, WavePlayer_5, WavePlayer_10, WavePlayer_11, WavePlayer_14]
  properties {weights [1,1,1,1,1,1,1,1]}
 }
 {id=WavePlayer_1, properties {soundWaveAssetPtr=&/Spot_Ice_Pressure_Close_01}}
 {id=WavePlayer_10, properties {soundWaveAssetPtr=&/Spot_Ice_Pressure_Close_06}}
 {id=WavePlayer_11, properties {soundWaveAssetPtr=&/Spot_Ice_Pressure_Close_07}}
 {id=WavePlayer_14, properties {soundWaveAssetPtr=&/Spot_Ice_Pressure_Close_08}}
 {id=WavePlayer_2, properties {soundWaveAssetPtr=&/Spot_Ice_Pressure_Close_02}}
 {id=WavePlayer_3, properties {soundWaveAssetPtr=&/Spot_Ice_Pressure_Close_03}}
 {id=WavePlayer_4, properties {soundWaveAssetPtr=&/Spot_Ice_Pressure_Close_04}}
 {id=WavePlayer_5, properties {soundWaveAssetPtr=&/Spot_Ice_Pressure_Close_05}}
]
properties {
 attenuationSettings=/Game/Sounds/Attenuation/Fort_TileAmbient_OneShot_Linear_Att
 maxDistance=8000
 soundClassObject=&Fort_Ambient_OneShots_Outdoors
 volumeMultiplier=0.351
 preEffectBusSends [{}]
 soundSubmixSends [{minSendLevel=1, soundSubmix=&Fortnite_Reverb_Distant}]
}
