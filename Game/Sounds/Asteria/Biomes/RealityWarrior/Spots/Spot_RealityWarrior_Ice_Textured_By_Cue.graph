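# Sound cue for one-shot textured ice spot ambience in the RealityWarrior biome of Asteria.
# A quality-level root feeds the same Modulator branch to its first two slots and leaves the other two empty.
# The Modulator wraps an equal-weight Random node, excluded from branch culling, that picks one of five wave variants.
# Output uses wind one-shot attenuation, the ambient one-shots sound class and a send to the exterior ambient source bus.
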SoundCue /Game/Sounds/Asteria/Biomes/RealityWarrior/Spots/Spot_RealityWarrior_Ice_Textured_By_Cue
root=QualityLevel_1
nodes [
 {id=Modulator_5, children [Random_6]}
 {id=QualityLevel_1, children [Modulator_5,Modulator_5,null,null]}
 {
  id=Random_6
  children [WavePlayer_20, WavePlayer_27, WavePlayer_28, WavePlayer_29, WavePlayer_30]
  properties {bShouldExcludeFromBranchCulling=true, weights [1,1,1,1,1]}
 }
 {id=WavePlayer_20, properties {soundWaveAssetPtr=&/Spot_RealityWarrior_Ice_Textured_By_01}}
 {id=WavePlayer_27, properties {soundWaveAssetPtr=&/Spot_RealityWarrior_Ice_Textured_By_02}}
 {id=WavePlayer_28, properties {soundWaveAssetPtr=&/Spot_RealityWarrior_Ice_Textured_By_03}}
 {id=WavePlayer_29, properties {soundWaveAssetPtr=&/Spot_RealityWarrior_Ice_Textured_By_04}}
 {id=WavePlayer_30, properties {soundWaveAssetPtr=&/Spot_RealityWarrior_Ice_Textured_By_06}}
]
properties {
 attenuationSettings=/Game/Sounds/Attenuation/Fort_TileAmbient_OneShot_Wind_Att
 maxDistance=8000
 soundClassObject=&Fort_Ambient_OneShots
 volumeMultiplier=0.401
 busSends [{audioBus=null, minSendLevel=0.5, sendLevel=1, soundSourceBus=&AmbientAudio_Exterior_SourceBus}]
}
