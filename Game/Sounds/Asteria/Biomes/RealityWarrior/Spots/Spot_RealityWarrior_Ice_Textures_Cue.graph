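# Sound cue for icy snow texture spot ambience in the RealityWarrior biome, played as a quiet outdoor ambient one-shot.
# A random node picks among nineteen equally weighted ice texture waves and feeds a modulator for variation.
# The quality-level root routes that chain to its first two slots and leaves the other two empty, so those levels play nothing.
# It uses wind one-shot tile attenuation and sends to the exterior ambient source bus.
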SoundCue /Game/Sounds/Asteria/Biomes/RealityWarrior/Spots/Spot_RealityWarrior_Ice_Textures_Cue
root=QualityLevel_1
nodes [
 {id=Modulator_5, children [Random_0]}
 {id=QualityLevel_1, children [Modulator_5,Modulator_5,null,null]}
 {
  id=Random_0
  children [
   WavePlayer_0
   WavePlayer_1
   WavePlayer_2
   WavePlayer_3
   WavePlayer_4
   WavePlayer_5
   WavePlayer_6
   WavePlayer_7
   WavePlayer_8
   WavePlayer_9
   WavePlayer_10
   WavePlayer_11
   WavePlayer_12
   WavePlayer_13
   WavePlayer_14
   WavePlayer_15
   WavePlayer_16
   WavePlayer_17
   WavePlayer_18
  ]
  properties {weights [1,1,1,1,1,1,1,1,1,1,1,1,1,1,1,1,1,1,1]}
 }
 {id=WavePlayer_0, properties {soundWaveAssetPtr=&/Spot_Snow_Ice_Texture_01}}
 {id=WavePlayer_1, properties {soundWaveAssetPtr=&/Spot_Snow_Ice_Texture_02}}
 {id=WavePlayer_10, properties {soundWaveAssetPtr=&/Spot_Snow_Ice_Texture_11}}
 {id=WavePlayer_11, properties {soundWaveAssetPtr=&/Spot_Snow_Ice_Texture_12}}
 {id=WavePlayer_12, properties {soundWaveAssetPtr=&/Spot_Snow_Ice_Texture_13}}
 {id=WavePlayer_13, properties {soundWaveAssetPtr=&/Spot_Snow_Ice_Texture_14}}
 {id=WavePlayer_14, properties {soundWaveAssetPtr=&/Spot_Snow_Ice_Texture_15}}
 {id=WavePlayer_15, properties {soundWaveAssetPtr=&/Spot_Snow_Ice_Texture_16}}
 {id=WavePlayer_16, properties {soundWaveAssetPtr=&/Spot_Snow_Ice_Texture_17}}
 {id=WavePlayer_17, properties {soundWaveAssetPtr=&/Spot_Snow_Ice_Texture_18}}
 {id=WavePlayer_18, properties {soundWaveAssetPtr=&/Spot_Snow_Ice_Texture_19}}
 {id=WavePlayer_2, properties {soundWaveAssetPtr=&/Spot_Snow_Ice_Texture_03}}
 {id=WavePlayer_3, properties {soundWaveAssetPtr=&/Spot_Snow_Ice_Texture_04}}
 {id=WavePlayer_4, properties {soundWaveAssetPtr=&/Spot_Snow_Ice_Texture_05}}
 {id=WavePlayer_5, properties {soundWaveAssetPtr=&/Spot_Snow_Ice_Texture_06}}
 {id=WavePlayer_6, properties {soundWaveAssetPtr=&/Spot_Snow_Ice_Texture_07}}
 {id=WavePlayer_7, properties {soundWaveAssetPtr=&/Spot_Snow_Ice_Texture_08}}
 {id=WavePlayer_8, properties {soundWaveAssetPtr=&/Spot_Snow_Ice_Texture_09}}
 {id=WavePlayer_9, properties {soundWaveAssetPtr=&/Spot_Snow_Ice_Texture_10}}
]
properties {
 attenuationSettings=/Game/Sounds/Attenuation/Fort_TileAmbient_OneShot_Wind_Att
 maxDistance=8000
 soundClassObject=&Fort_Ambient_OneShots_Outdoors
 volumeMultiplier=0.251
 busSends [{audioBus=null, minSendLevel=0.6, sendLevel=1, soundSourceBus=&AmbientAudio_Exterior_SourceBus}]
}
