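# Sound cue for a one-shot outdoor snow wind spot in the RealityWarrior biome, using heavy wind and gust waves from the Artemis snow set.
# A quality node routes its first two levels to five-variant pools and its last two to three-variant subsets.
# Each branch picks evenly between heavy wind and gusts, then picks a random wave with slight pitch and volume modulation.
# Wind one-shot attenuation and an outdoor weather sound class apply.
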SoundCue /Game/Sounds/Asteria/Biomes/RealityWarrior/Spots/Spot_RealityWarrior_Snow_Wind_Heavy_Cue
root=QualityLevel_1
nodes [
 {id=Modulator_0, children [Random_3], properties {pitchMax=1.1, pitchMin=0.9, volumeMax=1.1, volumeMin=0.9}}
 {id=Modulator_1, children [Random_5], properties {pitchMax=1.1, pitchMin=0.9, volumeMax=1.1, volumeMin=0.9}}
 {id=Modulator_3, children [Random_0], properties {pitchMax=1.1, pitchMin=0.9, volumeMax=1.1, volumeMin=0.9}}
 {id=Modulator_5, children [Random_1], properties {pitchMax=1.1, pitchMin=0.9, volumeMax=1.1, volumeMin=0.9}}
 {id=QualityLevel_1, children [Random_2,Random_2,Random_6,Random_6]}
 {id=Random_0, children [WavePlayer_0,WavePlayer_1,WavePlayer_2,WavePlayer_3,WavePlayer_4], properties {weights [1,1,1,1,1]}}
 {id=Random_1, children [WavePlayer_5,WavePlayer_6,WavePlayer_7], properties {weights [1,1,1]}}
 {id=Random_2, children [Modulator_3,Modulator_0], properties {weights [1,1]}}
 {id=Random_3, children [WavePlayer_8, WavePlayer_9, WavePlayer_10, WavePlayer_11, WavePlayer_12], properties {weights [1,1,1,1,1]}}
 {id=Random_5, children [WavePlayer_13, WavePlayer_14, WavePlayer_15], properties {weights [1,1,1]}}
 {id=Random_6, children [Modulator_5,Modulator_1], properties {weights [1,1]}}
 {id=WavePlayer_0, properties {soundWaveAssetPtr=/Game/Sounds/Artemis/Biomes/Snow/Spots/Spot_Snow_Wind_Heavy_03}}
 {id=WavePlayer_1, properties {soundWaveAssetPtr=/Game/Sounds/Artemis/Biomes/Snow/Spots/Spot_Snow_Wind_Heavy_04}}
 {id=WavePlayer_10, properties {soundWaveAssetPtr=/Game/Sounds/Artemis/Biomes/Snow/Spots/Spot_Snow_Wind_Gust_03}}
 {id=WavePlayer_11, properties {soundWaveAssetPtr=/Game/Sounds/Artemis/Biomes/Snow/Spots/Spot_Snow_Wind_Gust_04}}
 {id=WavePlayer_12, properties {soundWaveAssetPtr=/Game/Sounds/Artemis/Biomes/Snow/Spots/Spot_Snow_Wind_Gust_05}}
 {id=WavePlayer_13, properties {soundWaveAssetPtr=/Game/Sounds/Artemis/Biomes/Snow/Spots/Spot_Snow_Wind_Gust_01}}
 {id=WavePlayer_14, properties {soundWaveAssetPtr=/Game/Sounds/Artemis/Biomes/Snow/Spots/Spot_Snow_Wind_Gust_02}}
 {id=WavePlayer_15, properties {soundWaveAssetPtr=&Spot_Snow_Wind_Gust_03}}
 {id=WavePlayer_2, properties {soundWaveAssetPtr=/Game/Sounds/Artemis/Biomes/Snow/Spots/Spot_Snow_Wind_Heavy_05}}
 {id=WavePlayer_3, properties {soundWaveAssetPtr=/Game/Sounds/Artemis/Biomes/Snow/Spots/Spot_Snow_Wind_Heavy_01}}
 {id=WavePlayer_4, properties {soundWaveAssetPtr=/Game/Sounds/Artemis/Biomes/Snow/Spots/Spot_Snow_Wind_Heavy_02}}
 {id=WavePlayer_5, properties {soundWaveAssetPtr=&Spot_Snow_Wind_Heavy_03}}
 {id=WavePlayer_6, properties {soundWaveAssetPtr=&Spot_Snow_Wind_Heavy_04}}
 {id=WavePlayer_7, properties {soundWaveAssetPtr=&Spot_Snow_Wind_Heavy_05}}
 {id=WavePlayer_8, properties {soundWaveAssetPtr=&Spot_Snow_Wind_Gust_01}}
 {id=WavePlayer_9, properties {soundWaveAssetPtr=&Spot_Snow_Wind_Gust_02}}
]
properties {
 attenuationSettings=/Game/Sounds/Attenuation/Fort_TileAmbient_OneShot_Wind_Att
 bEnableBusSends=false
 maxDistance=8000
 soundClassObject=/Game/Sounds/FortSoundClassesAndModes/Fort_Ambient_Outdoors_Weather_Athena
 volumeMultiplier=0.251
 preEffectBusSends [{}]
}
