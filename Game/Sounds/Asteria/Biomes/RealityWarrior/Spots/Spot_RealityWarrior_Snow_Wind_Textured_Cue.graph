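# Sound cue for textured snow wind ambient spot one-shots in the RealityWarrior biome, reusing Artemis snow biome waves.
# A quality-level switch sends its first two slots to a random pick of fourteen variations and its last two to a reduced pool of four.
# Each pool feeds its own modulator for slight pitch and volume variation.
# Playback uses wind one-shot tile attenuation and the outdoor weather ambient sound class.
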SoundCue /Game/Sounds/Asteria/Biomes/RealityWarrior/Spots/Spot_RealityWarrior_Snow_Wind_Textured_Cue
root=QualityLevel_1
nodes [
 {id=Modulator_3, children [Random_0], properties {pitchMax=1.1, pitchMin=0.9, volumeMax=1.2, volumeMin=0.9}}
 {id=Modulator_5, children [Random_1], properties {pitchMax=1.1, pitchMin=0.9, volumeMax=1.1, volumeMin=0.9}}
 {id=QualityLevel_1, children [Modulator_3,Modulator_3,Modulator_5,Modulator_5]}
 {
  id=Random_0
  children [
   WavePlayer_9
   WavePlayer_10
   WavePlayer_11
   WavePlayer_12
   WavePlayer_13
   WavePlayer_14
   WavePlayer_15
   WavePlayer_16
   WavePlayer_17
   WavePlayer_0
   WavePlayer_1
   WavePlayer_2
   WavePlayer_3
   WavePlayer_4
  ]
  properties {weights [1,1,1,1,1,1,1,1,1,1,1,1,1,1]}
 }
 {id=Random_1, children [WavePlayer_18, WavePlayer_19, WavePlayer_20, WavePlayer_21], properties {weights [1,1,1,1]}}
 {id=WavePlayer_0, properties {soundWaveAssetPtr=/Game/Sounds/Artemis/Biomes/Snow/Spots/Spot_Snow_Wind_Textured_10}}
 {id=WavePlayer_1, properties {soundWaveAssetPtr=/Game/Sounds/Artemis/Biomes/Snow/Spots/Spot_Snow_Wind_Textured_11}}
 {id=WavePlayer_10, properties {soundWaveAssetPtr=/Game/Sounds/Artemis/Biomes/Snow/Spots/Spot_Snow_Wind_Textured_02}}
 {id=WavePlayer_11, properties {soundWaveAssetPtr=/Game/Sounds/Artemis/Biomes/Snow/Spots/Spot_Snow_Wind_Textured_03}}
 {id=WavePlayer_12, properties {soundWaveAssetPtr=/Game/Sounds/Artemis/Biomes/Snow/Spots/Spot_Snow_Wind_Textured_04}}
 {id=WavePlayer_13, properties {soundWaveAssetPtr=/Game/Sounds/Artemis/Biomes/Snow/Spots/Spot_Snow_Wind_Textured_05}}
 {id=WavePlayer_14, properties {soundWaveAssetPtr=/Game/Sounds/Artemis/Biomes/Snow/Spots/Spot_Snow_Wind_Textured_06}}
 {id=WavePlayer_15, properties {soundWaveAssetPtr=/Game/Sounds/Artemis/Biomes/Snow/Spots/Spot_Snow_Wind_Textured_07}}
 {id=WavePlayer_16, properties {soundWaveAssetPtr=/Game/Sounds/Artemis/Biomes/Snow/Spots/Spot_Snow_Wind_Textured_08}}
 {id=WavePlayer_17, properties {soundWaveAssetPtr=/Game/Sounds/Artemis/Biomes/Snow/Spots/Spot_Snow_Wind_Textured_09}}
 {id=WavePlayer_18, properties {soundWaveAssetPtr=&Spot_Snow_Wind_Textured_03}}
 {id=WavePlayer_19, properties {soundWaveAssetPtr=&Spot_Snow_Wind_Textured_02}}
 {id=WavePlayer_2, properties {soundWaveAssetPtr=/Game/Sounds/Artemis/Biomes/Snow/Spots/Spot_Snow_Wind_Textured_12}}
 {id=WavePlayer_20, properties {soundWaveAssetPtr=&Spot_Snow_Wind_Textured_04}}
 {id=WavePlayer_21, properties {soundWaveAssetPtr=&Spot_Snow_Wind_Textured_06}}
 {id=WavePlayer_3, properties {soundWaveAssetPtr=/Game/Sounds/Artemis/Biomes/Snow/Spots/Spot_Snow_Wind_Textured_13}}
 {id=WavePlayer_4, properties {soundWaveAssetPtr=/Game/Sounds/Artemis/Biomes/Snow/Spots/Spot_Snow_Wind_Textured_14}}
 {id=WavePlayer_9, properties {soundWaveAssetPtr=/Game/Sounds/Artemis/Biomes/Snow/Spots/Spot_Snow_Wind_Textured_01}}
]
properties {
 attenuationSettings=/Game/Sounds/Attenuation/Fort_TileAmbient_OneShot_Wind_Att
 bEnableBusSends=false
 maxDistance=8000
 soundClassObject=/Game/Sounds/FortSoundClassesAndModes/Fort_Ambient_Outdoors_Weather_Athena
 volumeMultiplier=0.501
 preEffectBusSends [{}]
}
